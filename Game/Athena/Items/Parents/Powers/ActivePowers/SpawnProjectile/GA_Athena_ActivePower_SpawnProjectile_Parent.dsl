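; Parent gameplay ability for Athena active powers that spawn a projectile from a player's cast montage.
; InitAbility builds hit and explode effect specs and resets a DoOnce gating notify-driven commit and spawn.
; PowerSpawnProjectile spawns on authority from socket and aim with scalable pitch, speed and gravity, optionally adding an extra explosion.
; Children can override homing and trajectory stubs; montage end, interruption or commit failure ends the ability.
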(event EventCastMontageNotifyStart
  (params (NotifyTag :type Gameplay%20Tag%20Structure))
  ;[1 1 1] Event Montage Notify Start Do the commit then spawn Projectile. Check if we should commit for every projectile or if we should ignore the commit.
  (|CallParentFunction NotifyTag)
  (if (GameplayTags|Equal(GameplayTag) NotifyTag (Variables|PowerProjectileSpawn|GetT_SpawnProjectileNotify))
    (Utilities|DoOnceThenDo false
      (:DoOnce
        _)
      (:ThenDo
        (if (Variables|PowerAbilityCommit|GetCommitOnEveryProjectile)
          (Utilities|PowerCommitAbility (Variables|AbilityCommit|GetIgnoreAbilityCommit)
            (:Pass
              _)
            (:Fail
              (Development|PrintString "Power: Ability Commit Failed")
              (CallFunction|PowerCommitFailed))
            (:Ignored
              _))
          (else
            _))
        (Networking|ForceNetUpdate (Variables|Default|GetPlayerPawn))
        (CallFunction|PowerSpawnProjectile)))
    (elif (and (GameplayTags|Equal(GameplayTag) NotifyTag (Variables|PowerAnimationCast|GetT_AbilityEndCastNotify)) (GameplayTags|NotEqual(GameplayTag) :A (Variables|PowerAnimationCast|GetT_AbilityEndCastNotify)))
      (CallFunction|CastMontageNotifyEndAbility))))

(event Custom|ResetCommitDoOnce
  ;[1 1 1] Event Montage Notify Start Do the commit then spawn Projectile. Check if we should commit for every projectile or if we should ignore the commit.
  (Utilities|DoOnceThenDo false
    (:DoOnce
      _)
    (:ThenDo
      (if (Variables|PowerAbilityCommit|GetCommitOnEveryProjectile)
        (Utilities|PowerCommitAbility (Variables|AbilityCommit|GetIgnoreAbilityCommit)
          (:Pass
            _)
          (:Fail
            (Development|PrintString "Power: Ability Commit Failed")
            (CallFunction|PowerCommitFailed))
          (:Ignored
            _))
        (else
          _))
      (Networking|ForceNetUpdate (Variables|Default|GetPlayerPawn))
      (CallFunction|PowerSpawnProjectile))))

(event EventCastMontageNotifyEndAbility
  ;[1 1 1] MontageEndNotify - End Ability
  (|CallParentFunction)
  (Ability|EndAbility))

(event EventCastMontageCancelledOrInterrupted
  ;[1 1 1] Montage Cancelled or Interrupted - end ability
  (|CallParentFunction)
  (Ability|EndAbility))

(event EventPowerCommitFailed
  ;[1 1 1] Commit Failed - End Ability
  (|CallParentFunction)
  (Ability|EndAbility))

(event EventCastMontageBlendOut
  ;[1 1 1] MontageBlendOut - End Ability
  (|CallParentFunction)
  (Ability|EndAbility))

(fn AddAdditionalExplosionContainer
  (params (Projectile :type Fort%20Projectile%20Base%20Object%20Reference))
  ;[1 1 1] Apply Additional Explosion Effect
  (bind _output_get (Variables|PowerProjectileSpawn|SetSpec_AdditionalExplode (Ability|MakeGameplayEffectContainerSpec (Variables|PowerProjectileSpawn|GetT_Proj_AdditionalExplodeApplicationTag) (Variables|PowerProjectileSpawn|GetProj_AdditionalExplodeEffectLevel))))
  (Delivery|AddAdditionalExplodeEffectContainer Projectile _output_get))

(fn GetHomingTarget
  ;[1 1 1] Override this in children to set homing targets.
  (return 0))

(fn GetProjectileSpawnDirection
  (params (Pawn :type Fort%20Player%20Pawn%20Object%20Reference))
  (bind (_returnvalue_roll _returnvalue_pitch _returnvalue_yaw) (Math|Rotator|BreakRotator (Targeting|GetDefaultTargetingRotation Pawn)))
  (return (Math|Rotator|MakeRotator _returnvalue_roll (+ _returnvalue_pitch (ScalableFloat|GetValueatLevel (Variables|PowerProjectileSpawn|GetRow_SpawnPitchOffset))) _returnvalue_yaw)))

(fn GetProjectileSpawnLocationAndRotation
  ;[1 1 1] Get the location and rotation to spawn the projectile based on player location and aiming.
  (bind _asfort_player_pawn (Utilities|Casting|CastToFortPlayerPawn (Ability|GetAvatarActorfromActorInfo))
    (:then
      (bind _direction (CallFunction|GetProjectileSpawnDirection _asfort_player_pawn))
      (return (Math|Vector|vector+vector (Transformation|GetSocketLocation (Variables|/Script/Engine.Character|GetMesh _asfort_player_pawn) (Variables|PowerProjectileSpawn|GetProj_SpawnSocket)) (* (Math|Vector|Normalize (Transformation|GetActorForwardVector _asfort_player_pawn)) (Variables|PowerProjectileSpawn|GetProj_SpawnForwardOffset)) (Variables|PowerProjectileSpawn|GetProj_SpawnWorldOffset)) _direction))
    (:CastFailed
      (Development|PrintString "Power: Spawn Projectile Avatar was not a Player Pawn. You need to Override the SpawnLocationAndRotation function."))))

(fn GetProjectileTrajectoryActor
  (return 0))

(fn GetProjectileTrajectoryPoints
  (return 0 0 0))

(fn InitAbility
  ;[1 1 1] Make the Effect Specs On Init so we can use them multiple times. After, reset the do once for commit
  (|CallParentFunction)
  (CallFunction|MakeEffectContainers)
  (CallFunction|ResetCommitDoOnce))

(fn MakeEffectContainers
  ;[1 1 1] Make Effect Containers to use on the projectile
  (bind _output_get (Variables|PowerProjectileSpawn|SetSpec_Hit (Ability|MakeGameplayEffectContainerSpec (Variables|PowerProjectileSpawn|GetT_Proj_HitApplicationTag) (Variables|PowerProjectileSpawn|GetProj_HitEffectLevel))))
  (bind _output_get_2 (Variables|PowerProjectileSpawn|SetSpec_Explode (Ability|MakeGameplayEffectContainerSpec (Variables|PowerProjectileSpawn|GetT_Proj_ExplodeApplicationTag) (Variables|PowerProjectileSpawn|GetProj_ExplodeEffectLevel))))
  (return _output_get _output_get_2))

(fn PowerSpawnProjectile
  (bind _returnvalue (Ability|GetAvatarActorfromActorInfo))
  (bind (_location _direction) (CallFunction|GetProjectileSpawnLocationandRotation))
  (bind _projectiletospawn (Variables|PowerProjectileSpawn|GetProjectileToSpawn))
  ;[1 1 1] On the server spawn the projectile.
  (Ability|HasAuthority
    (:True
      (Utilities|IsValid _returnvalue
        (:Is%20Valid
          (if (Utilities|IsValidClass _projectiletospawn)
            (CallFunction|PlayBurstGC (Variables|PowerGameplayCues|GetGC_SpawnProjectile) (Variables|PowerGameplayCues|GetSpawnProjectileGCUsesParams) (CallFunction|SpawnProjectileGCParams))
            (bind _returnvalue_2 (Projectile|SpawnProjectilewithParams _projectiletospawn _returnvalue (Utilities|Struct|MakeSpawnProjectileParams _location _direction _direction (Variables|PowerProjectileSpawn|GetSpec_Hit) (Variables|PowerProjectileSpawn|GetSpec_Explode) (Variables|PowerProjectileSpawn|GetProj_UseProjectileSpeedAndGrav) (CallFunction|GetHomingTarget) (ScalableFloat|GetValueatLevel (Variables|PowerProjectileSpawn|GetRow_ProjectileSpeed)) (ScalableFloat|GetValueatLevel (Variables|PowerProjectileSpawn|GetRow_ProjectileMaxSpeed)) (ScalableFloat|GetValueatLevel (Variables|PowerProjectileSpawn|GetRow_ProjectileGravity)) (select (Variables|PowerProjectileSpawn|GetProjUsesAbilityChargeMult) (Variables|PowerCharge|GetChargeMultiplier) (Variables|PowerProjectileSpawn|GetProj_ChargePercent)))))
            (bind _output_get (Variables|PowerProjectileSpawn|SetLastProjectileSpawned _returnvalue_2))
            ;;[1 1 1] Add Additional Explosion if desired
            (if (Variables|PowerProjectileSpawn|GetUsesAdditionalExplosionContainer)
              (CallFunction|AddAdditionalExplosionContainer _output_get)
              (else
                _))
            (Default|CallOnProjectileSpawned _output_get)
            (return _output_get)
            (else
              (Development|PrintString "Power: Did not spawn projectile Class was not valid")
              (return 0))))
        (:Is%20Not%20Valid
          (Development|PrintString "Power: Did not spawn projectile Avatar was not valid")
          (return 0))))))

(fn ShouldOnlyShowTrajectoryOnUse
  (return false))

(fn SpawnProjectileGCParams
  (return 0))

(fn SpawnTrajectoryIndicator
  (params (bSpawnOnEquip :type Boolean))
  (return 0))
